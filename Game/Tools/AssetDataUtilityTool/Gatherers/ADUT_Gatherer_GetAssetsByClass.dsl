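; Gatherer for the Asset Data Utility Tool that collects asset data by class from the Asset Registry.
; RunGatherer converts ReturnOnlyClasses to top-level class paths and queries the registry, including subclasses and in-memory assets.
; ConvertSoftClassToAssetPath splits soft class paths at the dot and accepts only native /Script/ classes, so Blueprint classes are skipped.
; FilterAssets applies the same native-class filter to a supplied asset array.
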(fn ConvertSoftClassToAssetPath
  (params (SoftClassReference :type Object%20Soft%20Class%20Reference))
  (bind _returnvalue (Utilities|ToSoftClassPath SoftClassReference))
  (bind _pathstring (Utilities|BreakSoftClassPath _returnvalue))
  (bind (_lefts _rights) (Utilities|String|Split _pathstring "."))
  (if (Utilities|String|StartsWith _pathstring "/Script/")
    (return (Utilities|MakeTopLevelAssetPath (Utilities|String|ToString(Name) (Utilities|String|StringToName _lefts)) (Utilities|String|ToString(Name) (Utilities|String|StringToName _rights))) true)
    (else
      (return 0 false))))

(fn FilterAssets
  (params (AssetData :type Array%20of%20Asset%20Data%20Structures) (Classes :type Array%20of%20Object%20Soft%20Class%20References))
  (LocalVariables|Setl_assets AssetData)
  (LocalVariables|Setl_classes Classes)
  (Utilities|FlowControl|Sequence
    (:then_0
      (for _array_element (LocalVariables|Getl_classes)
        (if (Utilities|String|StartsWith (Utilities|ToString(SoftClassReference) _array_element) "/Script/")
          (Utilities|Array|Add (LocalVariables|Getl_codeClasses) (CallFunction|ConvertSoftClasstoAssetPath _array_element)))))
    (:then_1
      (AssetRegistry|RunAssetsThroughFilter (AssetRegistry|GetAssetRegistry) (LocalVariables|Getl_assets) (Utilities|MakeARFilter :ClassPaths (LocalVariables|Getl_codeClasses) :bRecursivePaths false :bRecursiveClasses true :bIncludeOnlyOnDiskAssets false)))
    (:then_3
      (return (LocalVariables|Getl_assets)))))

(fn RunGatherer
  (Utilities|FlowControl|Sequence
    (:then_0
      (for _array_element (Variables|Default|GetReturnOnlyClasses)
        (if (.valid (CallFunction|ConvertSoftClasstoAssetPath _array_element))
          (Utilities|Array|Add (LocalVariables|Getl_classPaths) (CallFunction|ConvertSoftClasstoAssetPath _array_element)))))
    (:then_2
      (bind _outassetdata (AssetRegistry|GetAssets (AssetRegistry|GetAssetRegistry) (Utilities|MakeARFilter :ClassPaths (LocalVariables|Getl_classPaths) :bRecursivePaths false :bRecursiveClasses true :bIncludeOnlyOnDiskAssets false) false))
      (return _outassetdata))))
